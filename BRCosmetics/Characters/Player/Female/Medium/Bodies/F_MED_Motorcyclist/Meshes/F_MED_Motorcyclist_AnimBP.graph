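# Animation Blueprint for the F_MED_Motorcyclist cosmetic body, copying its pose from the attached parent mesh and adding deformation correctives.
# The copied pose runs through two Control Rigs in series, a twist rig and then a sculpter rig with exposed front-left knee corrective inputs, before reaching the output.
# Both rigs are borrowed from the F_MED_Textile_Sparkle body.
# The event graph holds no logic.
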AnimBlueprint /BRCosmetics/Characters/Player/Female/Medium/Bodies/F_MED_Motorcyclist/Meshes/F_MED_Motorcyclist_AnimBP
target_skeleton=&/F_MED_Motorcyclist_Skeleton
graphs [
 AnimGraph {nodes [
   n1 = Control Rig(Source=Copy Pose From Mesh() {bUseAttachedParent=true}) {
    alphaBoolBlend {blendOption=Linear}
    controlRigAssetReference {blueprintRigClass=/BRCosmetics/Characters/Player/Female/Medium/Bodies/F_MED_Textile_Sparkle/Meshes/F_MED_Textile_Sparkle_Deform_CtrlRig_twists.F_MED_Textile_Sparkle_Deform_CtrlRig_twists_C}
   }
   n2 = Control Rig!(
    Source=@n1
    Alpha
    KneeRot_fr_l_114deg:Vector
    KneeTrans_fr_l_114deg:Vector="0.258,-3.591,0.009"
    KneeScale_fr_l_114deg:Vector="1,1,1"
    KneeScale_fr_l_130deg:Vector="1,1,1"
    KneeRot_fr_l_130deg:Vector="0,0,-51.2"
    KneeTrans_fr_l_130deg:Vector="0.258,-3.591,0.009"
   ) -> (Pose) {
    alphaBoolBlend {blendOption=Linear}
    controlRigAssetReference {blueprintRigClass=/BRCosmetics/Characters/Player/Female/Medium/Bodies/F_MED_Textile_Sparkle/Meshes/F_MED_Textile_Sparkle_Deform_CtrlRig_sculpters.F_MED_Textile_Sparkle_Deform_CtrlRig_sculpters_C}
   }
   Output Pose(Result=@n2) {layerGroup=None}
  ]}
 EventGraph {nodes [Event BlueprintUpdateAnimation()]}
]
variables []
