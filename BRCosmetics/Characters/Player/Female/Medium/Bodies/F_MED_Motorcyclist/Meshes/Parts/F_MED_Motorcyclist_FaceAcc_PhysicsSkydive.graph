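# Physics asset for the female medium Motorcyclist face accessory mesh, a skydive variant that drives hair-bone motion.
# Capsules cover head, neck, upper spine, clavicles and three-segment hair chains (A–D on each side, E in the center), all rooted at the head.
# Chains use narrow angular limits, tighter and twist-locked for C, D and E.
# Fully free constraints link the body capsules and some C, D and E segments, asymmetrically between sides.
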PhysicsAsset /BRCosmetics/Characters/Player/Female/Medium/Bodies/F_MED_Motorcyclist/Meshes/Parts/F_MED_Motorcyclist_FaceAcc_PhysicsSkydive
bodies {
 clavicle_l [{shapeType=Capsule, length=27, radius=4.75, shapeName=clavicle_r_capsule, center=<7.7633 -1.445 -0.6446>, rotation=<rot 0 -89.9998 -79.9972>}]
 clavicle_r [{
   shapeType=Capsule
   length=23.75
   radius=5.75
   shapeName=clavicle_r_capsule
   center=<-11.7098 1.2284 1.5559>
   rotation=<rot -0.0001 -89.9996 -79.9957>
  }]
 hairA1_l [{shapeType=Capsule, length=3, radius=2, shapeName=hairA1_l_capsule, center=<2 0 1.75>, rotation=<rot 90 0 0>}]
 hairA1_r [{shapeType=Capsule, length=3, radius=2, shapeName=hairA1_r_capsule, center=<2 0 0>, rotation=<rot 90 0 0>}]
 hairA2_l [{shapeType=Capsule, length=3.25, radius=4, shapeName=hairA2_l_capsule, center=<2 3 0>, rotation=<rot 90 0 0>}]
 hairA2_r [{shapeType=Capsule, length=3, radius=2, shapeName=hairA2_r_capsule, center=<2 0 0>, rotation=<rot 90 0 0>}]
 hairA3_l [{shapeType=Capsule, length=3.25, radius=4, shapeName=hairA3_l_capsule, center=<-0.3224 3 0>, rotation=<rot 90 0 0>}]
 hairA3_r [{shapeType=Capsule, length=3, radius=2, shapeName=hairA3_r_capsule, center=<-0.0951 0 0.0001>, rotation=<rot 90 0 0>}]
 hairB1_l [{shapeType=Capsule, length=3, radius=2, shapeName=hairB1_l_capsule, center=<2 0 0>, rotation=<rot 90 0 0>}]
 hairB1_r [{shapeType=Capsule, length=3, radius=2, shapeName=hairB1_r_capsule, center=<2 0 0>, rotation=<rot 90 0 0>}]
 hairB2_l [{shapeType=Capsule, length=4.5, radius=4.25, shapeName=hairB2_l_capsule, center=<5.2298 11 0>, rotation=<rot 90 0 0>}]
 hairB2_r [{shapeType=Capsule, length=3, radius=2, shapeName=hairB2_r_capsule, center=<2 0 0>, rotation=<rot 90 0 0>}]
 hairB3_l [{shapeType=Capsule, length=4.5, radius=4.25, shapeName=hairB3_l_capsule, center=<4.6339 11 0>, rotation=<rot 90 0 0>}]
 hairB3_r [{shapeType=Capsule, length=3, radius=2, shapeName=hairB3_r_capsule, center=<1.5982 0 0>, rotation=<rot 90 0 0>}]
 hairC1_l [{shapeType=Capsule, length=3, radius=2, shapeName=hairC1_l_capsule, center=<3.9018 0 0>, rotation=<rot 90 0 0>}]
 hairC1_r [{shapeType=Capsule, length=3, radius=2, shapeName=hairC1_r_capsule, center=<4.9288 0 0>, rotation=<rot 90 0 0>}]
 hairC2_l [{shapeType=Capsule, length=3, radius=2, shapeName=hairC2_l_capsule, center=<2 0 0>, rotation=<rot 90 0 0>}]
 hairC2_r [{shapeType=Capsule, length=6, radius=3.75, shapeName=hairC2_r_capsule, center=<2.7684 5.7714 -3>, rotation=<rot 90 14.0363 -5.9686>}]
 hairC3_l [{shapeType=Capsule, length=3, radius=2, shapeName=hairC3_l_capsule, center=<1.0531 0 -0.0002>, rotation=<rot 90 0 0>}]
 hairC3_r [{shapeType=Capsule, length=7.5, radius=4.5, shapeName=hairC3_r_capsule, center=<9.1025 11.2614 -3>, rotation=<rot 90 -12.9074 -22.9123>}]
 hairD1_l [{shapeType=Capsule, length=3, radius=2, shapeName=hairD1_l_capsule, center=<4.1344 0 0>, rotation=<rot 90 0 0>}]
 hairD1_r [{shapeType=Capsule, length=3, radius=2, shapeName=hairD1_r_capsule, center=<4.3466 0 0>, rotation=<rot 90 0 0>}]
 hairD2_l [{shapeType=Capsule, length=5, radius=5, shapeName=hairD2_l_capsule, center=<2 0 4.5>, rotation=<rot 90 0 0>}]
 hairD2_r [{shapeType=Capsule, length=5, radius=4.75, shapeName=hairD2_r_capsule, center=<1.8224 2.0306 0.968>, rotation=<rot 70.0768 -174.923 174.9231>}]
 hairD3_l [{shapeType=Capsule, length=5, radius=5, shapeName=hairD3_l_capsule, center=<1.7022 0 4.5>, rotation=<rot 90 0 0>}]
 hairD3_r [{
   shapeType=Capsule
   length=4.25
   radius=5
   shapeName=hairD3_r_capsule
   center=<5.6231 4.6249 4.0236>
   rotation=<rot 69.9992 -179.9999 169.9984>
  }]
 hairE1_ct [{shapeType=Capsule, length=3, radius=2, shapeName=hairE1_ct_capsule, center=<4.6512 0 0>, rotation=<rot 90 0 0>}]
 hairE2_ct [{shapeType=Capsule, length=5.25, radius=4.25, shapeName=hairE2_ct_capsule, center=<2 0 2.25>, rotation=<rot 90 0 0>}]
 hairE3_ct [{shapeType=Capsule, length=5.25, radius=4.25, shapeName=hairE3_ct_capsule, center=<2 0 4>, rotation=<rot 90 0 0>}]
 head [{shapeType=Capsule, length=4.25, radius=6.5, shapeName=head_capsule, center=<3.4362 -0.0001 0>, rotation=<rot 90 0 0>}]
 neck_01 [{shapeType=Capsule, length=5, radius=5.25, shapeName=neck_01_capsule, rotation=<rot -90 0 -10.0021>}]
 neck_02 [{shapeType=Capsule, length=6, radius=4, shapeName=neck_02_capsule, rotation=<rot -90 -0.633 0.6329>}]
 spine_05 [{shapeType=Capsule, length=11.75, radius=8.255, shapeName=spine_05_capsule, center=<2.1817 -0.9815 0>, rotation=<rot 0 -89.9999 0>}]
}
constraints [
 {
  bone1Name=head
  bone2Name=neck_02
  swing1LimitDegrees=45
  swing1Motion=Free
  swing2LimitDegrees=45
  swing2Motion=Free
  twistLimitDegrees=45
  twistMotion=Free
 }
 {
  bone1Name=neck_02
  bone2Name=neck_01
  swing1LimitDegrees=45
  swing1Motion=Free
  swing2LimitDegrees=45
  swing2Motion=Free
  twistLimitDegrees=45
  twistMotion=Free
 }
 {
  bone1Name=clavicle_l
  bone2Name=spine_05
  swing1LimitDegrees=45
  swing1Motion=Free
  swing2LimitDegrees=45
  swing2Motion=Free
  twistLimitDegrees=45
  twistMotion=Free
 }
 {
  bone1Name=clavicle_r
  bone2Name=spine_05
  swing1LimitDegrees=45
  swing1Motion=Free
  swing2LimitDegrees=45
  swing2Motion=Free
  twistLimitDegrees=45
  twistMotion=Free
 }
 {
  bone1Name=neck_01
  bone2Name=spine_05
  swing1LimitDegrees=45
  swing1Motion=Free
  swing2LimitDegrees=45
  swing2Motion=Free
  twistLimitDegrees=45
  twistMotion=Free
 }
 {bone1Name=hairA1_l, bone2Name=head, swing1LimitDegrees=20, swing2LimitDegrees=20, twistLimitDegrees=3}
 {bone1Name=hairA2_l, bone2Name=hairA1_l, swing1LimitDegrees=20, swing2LimitDegrees=20, twistLimitDegrees=3}
 {bone1Name=hairA3_l, bone2Name=hairA2_l, swing1LimitDegrees=20, swing2LimitDegrees=20, twistLimitDegrees=3}
 {bone1Name=hairB1_l, bone2Name=head, swing1LimitDegrees=20, swing2LimitDegrees=20, twistLimitDegrees=3}
 {bone1Name=hairB2_l, bone2Name=hairB1_l, swing1LimitDegrees=20, swing2LimitDegrees=20, twistLimitDegrees=3}
 {bone1Name=hairB3_l, bone2Name=hairB2_l, swing1LimitDegrees=20, swing2LimitDegrees=20, twistLimitDegrees=3}
 {bone1Name=hairC2_l, bone2Name=hairC1_l, swing1LimitDegrees=10, swing2LimitDegrees=10, twistLimitDegrees=0}
 {bone1Name=hairC1_l, bone2Name=head, swing1LimitDegrees=10, swing2LimitDegrees=10, twistLimitDegrees=0}
 {bone1Name=hairC3_l, bone2Name=hairC2_l, swing1LimitDegrees=10, swing2LimitDegrees=10, twistLimitDegrees=0}
 {bone1Name=hairD1_l, bone2Name=head, swing1LimitDegrees=10, swing2LimitDegrees=10, twistLimitDegrees=0}
 {bone1Name=hairD2_l, bone2Name=hairD1_l, swing1LimitDegrees=10, swing2LimitDegrees=10, twistLimitDegrees=0}
 {bone1Name=hairD3_l, bone2Name=hairD2_l, swing1LimitDegrees=10, swing2LimitDegrees=10, twistLimitDegrees=0}
 {bone1Name=hairE1_ct, bone2Name=head, swing1LimitDegrees=10, swing2LimitDegrees=10, twistLimitDegrees=0}
 {bone1Name=hairE2_ct, bone2Name=hairE1_ct, swing1LimitDegrees=10, swing2LimitDegrees=10, twistLimitDegrees=0}
 {bone1Name=hairE3_ct, bone2Name=hairE2_ct, swing1LimitDegrees=10, swing2LimitDegrees=10, twistLimitDegrees=0}
 {bone1Name=hairA1_r, bone2Name=head, swing1LimitDegrees=20, swing2LimitDegrees=20, twistLimitDegrees=3}
 {bone1Name=hairA2_r, bone2Name=hairA1_r, swing1LimitDegrees=20, swing2LimitDegrees=20, twistLimitDegrees=3}
 {bone1Name=hairA3_r, bone2Name=hairA2_r, swing1LimitDegrees=20, swing2LimitDegrees=20, twistLimitDegrees=3}
 {bone1Name=hairB1_r, bone2Name=head, swing1LimitDegrees=20, swing2LimitDegrees=20, twistLimitDegrees=3}
 {bone1Name=hairB2_r, bone2Name=hairB1_r, swing1LimitDegrees=20, swing2LimitDegrees=20, twistLimitDegrees=3}
 {bone1Name=hairB3_r, bone2Name=hairB2_r, swing1LimitDegrees=20, swing2LimitDegrees=20, twistLimitDegrees=3}
 {bone1Name=hairC1_r, bone2Name=head, swing1LimitDegrees=10, swing2LimitDegrees=10, twistLimitDegrees=0}
 {bone1Name=hairC2_r, bone2Name=hairC1_r, swing1LimitDegrees=10, swing2LimitDegrees=10, twistLimitDegrees=0}
 {bone1Name=hairC3_r, bone2Name=hairC2_r, swing1LimitDegrees=10, swing2LimitDegrees=10, twistLimitDegrees=0}
 {bone1Name=hairD1_r, bone2Name=head, swing1LimitDegrees=10, swing2LimitDegrees=10, twistLimitDegrees=0}
 {bone1Name=hairD2_r, bone2Name=hairD1_r, swing1LimitDegrees=10, swing2LimitDegrees=10, twistLimitDegrees=0}
 {bone1Name=hairD3_r, bone2Name=hairD2_r, swing1LimitDegrees=10, swing2LimitDegrees=10, twistLimitDegrees=0}
 {
  bone1Name=hairC2_l
  bone2Name=hairD2_l
  swing1LimitDegrees=45
  swing1Motion=Free
  swing2LimitDegrees=45
  swing2Motion=Free
  twistLimitDegrees=45
  twistMotion=Free
 }
 {
  bone1Name=hairC3_l
  bone2Name=hairD3_l
  swing1LimitDegrees=45
  swing1Motion=Free
  swing2LimitDegrees=45
  swing2Motion=Free
  twistLimitDegrees=45
  twistMotion=Free
 }
 {
  bone1Name=hairD2_l
  bone2Name=hairE2_ct
  swing1LimitDegrees=45
  swing1Motion=Free
  swing2LimitDegrees=45
  swing2Motion=Free
  twistLimitDegrees=45
  twistMotion=Free
 }
 {
  bone1Name=hairC3_r
  bone2Name=hairC2_r
  swing1LimitDegrees=45
  swing1Motion=Free
  swing2LimitDegrees=45
  swing2Motion=Free
  twistLimitDegrees=45
  twistMotion=Free
 }
 {
  bone1Name=hairC3_r
  bone2Name=hairD3_r
  swing1LimitDegrees=45
  swing1Motion=Free
  swing2LimitDegrees=45
  swing2Motion=Free
  twistLimitDegrees=45
  twistMotion=Free
 }
 {
  bone1Name=hairD2_r
  bone2Name=hairC2_r
  swing1LimitDegrees=45
  swing1Motion=Free
  swing2LimitDegrees=45
  swing2Motion=Free
  twistLimitDegrees=45
  twistMotion=Free
 }
]
