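# Post-process style hair rig for the F_MED_Motorcyclist FaceAcc mesh, copying the attached parent's pose and adding secondary motion.
# Two head-space RigidBody sims with separate physics assets swap by skydiving state, followed by rotation constraints between hair bones (two at zero alpha).
# Additive bone offsets before and after physics shift right-side hair for aiming, crouched aiming and head rotation.
# The event graph derives these alphas from character state and head-versus-neck rotation.
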AnimBlueprint /BRCosmetics/Characters/Player/Female/Medium/Bodies/F_MED_Motorcyclist/Meshes/Parts/F_MED_Motorcyclist_FaceAcc_AnimBP
target_skeleton=&/F_MED_Motorcyclist_FaceAcc_Skeleton
graphs [
 AnimGraph {nodes [
   n1 = "Transform (Modify) Bone - Bone: hairC1_r"(
    ComponentPose=Local To Component(LocalPose=Copy Pose From Mesh() {bUseAttachedParent=true})
    Alpha=Get ADS() -> (ADS:double)
    Translation="0,5,5"
    Rotation="0,0,0"
    Scale="1,1,1"
   ) {
    rotationMode=BMM_Additive
    rotationSpace=BCS_BoneSpace
    scaleMode=BMM_Additive
    scaleSpace=BCS_BoneSpace
    translationMode=BMM_Additive
    translationSpace=BCS_BoneSpace
    alphaBoolBlend {blendOption=Linear}
    boneToModify {boneName=hairC1_r}
    translation {y=5, z=5}
   }
   n2 = RigidBody(ComponentPose=@n1, Alpha=Get isNotSkydiving() -> (isNotSkydiving:double)) {
    bClampLinearTranslationLimitToRefPose=true
    bOverrideWorldGravity=true
    overridePhysicsAsset=&/F_MED_Motorcyclist_FaceAcc_PhysicsDuplicated
    simulationSpace=BaseBoneSpace
    alphaBoolBlend {blendOption=Linear}
    baseBoneRef {boneName=head}
    componentAppliedLinearAccClamp=<3000 3000 3000>
    componentLinearAccScale=<0.4 0.3 0.3>
    componentLinearVelScale=<0.4 0.3 0.3>
    overrideWorldGravity {z=-980}
   }
   n3 = RigidBody(ComponentPose=@n2, Alpha=Get isSkydiving() -> (isSkydiving:double)) {
    bClampLinearTranslationLimitToRefPose=true
    bOverrideWorldGravity=true
    overridePhysicsAsset=&/F_MED_Motorcyclist_FaceAcc_PhysicsSkydive
    simulationSpace=BaseBoneSpace
    alphaBoolBlend {blendOption=Linear}
    baseBoneRef {boneName=head}
    componentAppliedLinearAccClamp=<3000 3000 3000>
    componentLinearAccScale=<0.4 0.3 0.3>
    componentLinearVelScale=<0.4 0.3 0.3>
    overrideWorldGravity {z=-980}
   }
   n4 = "Constraint - hairD1_r (R)"(ComponentPose=@n3, Alpha=0, ConstraintWeights_0=1) {
    alphaBoolBlend {blendOption=Linear}
    boneToModify {boneName=hairD1_r}
    constraintSetup [{transformType=Rotation, targetBone {boneName=hairC1_r}}]
    constraintWeights [1]
   }
   n5 = "Constraint - hairD2_r (R)"(ComponentPose=@n4, Alpha=0, ConstraintWeights_0=1) {
    alphaBoolBlend {blendOption=Linear}
    boneToModify {boneName=hairD2_r}
    constraintSetup [{transformType=Rotation, targetBone {boneName=hairD2_l}}]
    constraintWeights [1]
   }
   n6 = "Constraint - hairD3_r (R)"(ComponentPose=@n5, Alpha=0.2, ConstraintWeights_0=1) {
    alphaBoolBlend {blendOption=Linear}
    boneToModify {boneName=hairD3_r}
    constraintSetup [{transformType=Rotation, targetBone {boneName=hairC3_r}}]
    constraintWeights [1]
   }
   n7 = "Constraint - hairE2_ct (RR)"!(ComponentPose=@n6, Alpha=1, ConstraintWeights_0=0.8, ConstraintWeights_1:float=0.2) -> (Pose) {
    alphaBoolBlend {blendOption=Linear}
    boneToModify {boneName=hairE2_ct}
    constraintSetup [{transformType=Rotation, targetBone {boneName=hairD2_r}}, {transformType=Rotation, targetBone {boneName=hairD2_l}}]
    constraintWeights [1,1]
   }
   n8 = "Transform (Modify) Bone - Bone: hairC1_r"(ComponentPose=@n7, Alpha=Get ADS(), Translation="0,1,0.5", Rotation="0,0,0", Scale="1,1,1") {
    rotationMode=BMM_Additive
    rotationSpace=BCS_BoneSpace
    scaleMode=BMM_Additive
    scaleSpace=BCS_BoneSpace
    translationMode=BMM_Additive
    translationSpace=BCS_BoneSpace
    alphaBoolBlend {blendOption=Linear}
    boneToModify {boneName=hairC1_r}
    translation {y=1, z=0.5}
   }
   n9 = "Transform (Modify) Bone - Bone: hairC1_r"(ComponentPose=@n8, Alpha=Get headRot() -> (headRot:double), Translation="0,5,5", Rotation="0,0,0", Scale="1,1,1") {
    rotationMode=BMM_Additive
    rotationSpace=BCS_BoneSpace
    scaleMode=BMM_Additive
    scaleSpace=BCS_BoneSpace
    translationMode=BMM_Additive
    translationSpace=BCS_BoneSpace
    alphaBoolBlend {blendOption=Linear}
    boneToModify {boneName=hairC1_r}
    translation {y=5, z=5}
   }
   n10 = Get ADS()
   n11 = "Transform (Modify) Bone - Bone: hairB2_r"(ComponentPose=@n9, Alpha=@n10, Translation="-2,0,0", Rotation="0,0,0", Scale="1,1,1") {
    rotationMode=BMM_Additive
    rotationSpace=BCS_BoneSpace
    scaleMode=BMM_Additive
    scaleSpace=BCS_BoneSpace
    translationMode=BMM_Additive
    translationSpace=BCS_BoneSpace
    alphaBoolBlend {blendOption=Linear}
    boneToModify {boneName=hairB2_r}
    translation {x=-2}
   }
   n12 = "Transform (Modify) Bone - Bone: hairA2_r"(ComponentPose=@n11, Alpha=@n10, Translation="-2,0,0", Rotation="0,0,0", Scale="1,1,1") {
    rotationMode=BMM_Additive
    rotationSpace=BCS_BoneSpace
    scaleMode=BMM_Additive
    scaleSpace=BCS_BoneSpace
    translationMode=BMM_Additive
    translationSpace=BCS_BoneSpace
    alphaBoolBlend {blendOption=Linear}
    boneToModify {boneName=hairA2_r}
    translation {x=-2}
   }
   n13 = Get CrouchADS() -> (CrouchADS:double)
   n14 = "Transform (Modify) Bone - Bone: hairC1_r"(ComponentPose=@n12, Alpha=@n13, Translation="-4.5,-2.5,2", Rotation="0,0,5", Scale="1,1,1") {
    rotationMode=BMM_Additive
    translationMode=BMM_Additive
    alphaBoolBlend {blendOption=Linear}
    alphaScaleBiasClamp {bInterpResult=true, interpSpeedDecreasing=1, interpSpeedIncreasing=1}
    boneToModify {boneName=hairC1_r}
    rotation {roll=5}
    translation=<-4.5 -2.5 2>
   }
   n15 = "Transform (Modify) Bone - Bone: hairD1_r"(ComponentPose=@n14, Alpha=@n13, Translation="-2.5,-1.25,2.5", Rotation="0,0,2", Scale="1,1,1") {
    rotationMode=BMM_Additive
    translationMode=BMM_Additive
    alphaBoolBlend {blendOption=Linear}
    alphaScaleBiasClamp {bInterpResult=true, interpSpeedDecreasing=1, interpSpeedIncreasing=1}
    boneToModify {boneName=hairD1_r}
    rotation {roll=2}
    translation=<-2.5 -1.25 2.5>
   }
   Output Pose(Result=Component To Local(ComponentPose=@n15)) {layerGroup=None}
  ]}
 EventGraph {nodes [
   Event BlueprintUpdateAnimation()
   > n1 = Sequence()
   n2 = Get bIsTargeting()
   > Set ADS(ADS:double="To Float (Boolean)"(InBool=@n2)) -> (Output_Get:double)
   > Set headRot(headRot:double=GetAlphaFromReferenceTransformDelta(SkeletalMeshComponent=GetOwnerSkeletalMeshComponent(), BoneName=head, BaseName=neck_01, InputRange=(X=0,Y=-100), TwistAxis="1,0,0")) -> (Output_Get:double)
   > Set isSkydiving(isSkydiving:double="To Float (Boolean)"(InBool=Get bIsSkyDiving())) -> (Output_Get:double)
   n3 = Get bIsSkyDiving()
   > Set isNotSkydiving(isNotSkydiving:double="To Float (Boolean)"(InBool=NOT Boolean(A=@n3))) -> (Output_Get:double)
   n4 = Set ADS_Crouching(execute=@n1.then_1, ADS_Crouching:bool=AND Boolean(A=@n2, B=Get bIsCrouching())) -> (Output_Get:bool)
   > Set CrouchADS(CrouchADS:double="To Float (Boolean)"(InBool=@n4)) -> (Output_Get:double)
   Get ADS()
  ]}
]
variables [
 ADS {type=number}
 ADS_Crouching {type=boolean}
 CrouchADS {type=number}
 headRot {type=number}
 isNotSkydiving {type=number}
 isSkydiving {type=number}
 NewVar_0 {type=boolean}
]
